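# Unlit translucent, two-sided smoke trail material for the SheerWill emote, built for Niagara sprite and mesh particles.
# World position offset combines SplineThicken, with width from a dynamic parameter shaped by a curve along U, and a per-particle randomized sine wave in a local direction.
# A dynamic parameter drives an 8x8 smoke flipbook, which modulates particle-colored emissive and supplies opacity.
# Opacity fades at the low-U end and uses depth fade.
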Material /BRCosmetics/Effects/Fort_Effects/Emotes/SheerWill/M_Emote_SheerWill_DetailSmokeTrail
graph {
 properties {
  bAutomaticallySetUsageInEditor=false
  bUsedWithNiagaraMeshParticles=true
  bUsedWithNiagaraSprites=true
  blendMode=BLEND_Translucent
  shadingModel=MSM_Unlit
  twoSided=true
  usedShadingModels=Unlit
 }
 nodes [
  n1 = TextureCoordinate()
  n2 = Multiply(
   A=Multiply(A=Constant() {r=70}, B=DynamicParameter() {paramNames [SplineWidth,Animation,Param3,Param4]}.SplineWidth)
   B=ColorRamp!(None=ComponentMask(None=@n1) {r=true}) {colorCurve="<self>.ColorCurve"}
  )
  n3 = StaticBool()
  n4 = ComponentMask(None=TextureCoordinate()) {r=true}
  n5 = ParticleRandom()
  n6 = Time()
  n7 = LinearInterpolate(Alpha=ParticleRandom()) {constA=-0.8, constB=-1.1}
  n8 = Sine(None=Add(A=Add(A=@n4, B=@n5), B=Multiply(A=@n6, B=@n7) {constB=-1}))
  n9 = Transform(None=VectorParameter() {parameterName=WPO_Direction, defaultValue=<rgba 5 0 0 0>}.RGB) {transformSourceType=TRANSFORMSOURCE_Local}
  LinearInterpolate(
   B=Add(
    A=MaterialFunctionCall(WidthBase=@n2, WidthTip=@n2, Expand U or V UV Channel=@n3) {materialFunction=&SplineThicken}.WorldPositionOffset
    B=Multiply(A=@n8, B=@n9)
   )
   Alpha=ScalarParameter() {defaultValue=1, parameterName=UseWPO}
  )
  n10 = ParticleColor()
  n11 = Constant() {r=8}
  n12 = MaterialFunctionCall(
   "Animation  Phase (0-1)"=DynamicParameter() {paramNames [SplineWidth,Animation,Param3,Param4]}.Animation
   Number of Rows=@n11
   Number of Columns=@n11
   Texture=TextureObjectParameter() {parameterName=DefaultTex, texture=&/T_Emote_SheerWill_Smoke_Trail}
  ) {materialFunction=&FlipBook}
  n13 = Power(
   Base=LinearInterpolate(A=@n12.Alpha, B=@n12.Alpha, Alpha=ScalarParameter() {parameterName=Use Result})
   Exp=ScalarParameter() {defaultValue=0.7, parameterName=Power}
  ) {constExponent=0.7}
  n14 = ScalarParameter() {defaultValue=0.2, parameterName=AlphaLerp}
  MaterialFunctionCall(Emissive=Multiply(A=@n10.RGB, B=LinearInterpolate(B=@n13, Alpha=@n14) {constA=1, constAlpha=0.2})) {materialFunction=&MF_FX_EmissiveAdjustments}
  n15 = Add(A=ComponentMask(None=TextureCoordinate()) {r=true}) {constB=-0.05}
  n16 = Multiply(A=Multiply(A=@n10.A, B=@n12.Alpha), B=Saturate(None=Multiply(A=@n15) {constB=5}))
  n17 = ScalarParameter() {defaultValue=1, parameterName=AlphaPower}
  Reroute(None=DepthFade(Opacity=Power(Base=@n16, Exp=@n17), FadeDistance=ScalarParameter() {defaultValue=10, parameterName=FadeDistance}) {fadeDistanceDefault=10})
 ]
}
