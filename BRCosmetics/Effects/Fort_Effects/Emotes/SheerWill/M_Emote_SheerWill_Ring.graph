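# Translucent, unlit, two-sided material for the ring effect in the SheerWill emote, used on Niagara mesh particles.
# World position offset combines SplineThicken, sized by the SplineWidth dynamic parameter, with panning noise wobble along a local-space WPO_Direction.
# Opacity blends a scrolling trail texture toward a sine-shaped band via Use_Glow, then applies particle alpha and depth fade.
# Emissive is particle color modulated by blurred noise through MF_FX_EmissiveAdjustments.
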Material /BRCosmetics/Effects/Fort_Effects/Emotes/SheerWill/M_Emote_SheerWill_Ring
graph {
 properties {
  bAutomaticallySetUsageInEditor=false
  bUsedWithNiagaraMeshParticles=true
  blendMode=BLEND_Translucent
  shadingModel=MSM_Unlit
  twoSided=true
  usedShadingModels=Unlit
 }
 nodes [
  n1 = Multiply(A=Constant() {r=5}, B=DynamicParameter() {paramNames [SplineWidth,Param2,Param3,Param4]}.SplineWidth)
  n2 = TextureSample(UVs=Panner(Coordinate=TextureCoordinate() {uTiling=2, vTiling=0}) {speedX=-0.3}) {samplerType=SAMPLERTYPE_Grayscale, texture=&LowResBlurredNoise}
  n3 = VectorParameter() {parameterName=WPO_Direction, defaultValue=<rgba 0 8 0 0>}
  Add(
   A=MaterialFunctionCall(WidthBase=@n1, WidthTip=@n1, Expand U or V UV Channel=StaticBool()) {materialFunction=&SplineThicken}.WorldPositionOffset
   B=Multiply(A=ConstantBiasScale(None=@n2.R) {bias=-1, scale=2}, B=Transform(None=@n3.RGB) {transformSourceType=TRANSFORMSOURCE_Local})
  )
  n4 = ParticleColor()
  n5 = TextureSampleParameter2D(UVs=Panner(Coordinate=TextureCoordinate() {uTiling=3}) {speedX=1}) {parameterName=Default_Tex, texture=/BRCosmetics/Weapons/FORT_Melee/Pickaxe_Tower_Sentinel_Female/FX/T_TowerSentinel_trail}
  n6 = TextureCoordinate()
  n7 = Power(Base=Sine(None=ComponentMask(None=@n6) {g=true}) {period=2}) {constExponent=3}
  n8 = ScalarParameter() {parameterName=Use_Glow}
  DepthFade(Opacity=Multiply(A=@n4.A, B=LinearInterpolate(A=@n5.R, B=@n7, Alpha=@n8))) {fadeDistanceDefault=5}
  n9 = Power(Base=TextureSample() {samplerType=SAMPLERTYPE_Grayscale, texture=&LowResBlurredNoise}.R) {constExponent=1.5}
  MaterialFunctionCall(Emissive=Multiply(A=@n4.RGB, B=LinearInterpolate(Alpha=@n9) {constA=0.3, constB=3})) {materialFunction=&MF_FX_EmissiveAdjustments}
 ]
}
